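# Niagara system for a damaged fence effect in the CrewCut set, built from Sparks, Airglow and Flare emitters.
# At system level, SystemState runs alongside two variable-assignment modules.
# A position array, SelectPositionFromArray and RandomRangeFloat inputs are used, suggesting the effect varies its location and a float value.
# It exposes no user variables, and no renderers are listed for the emitters.
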NiagaraSystem /Hermes/Sets/CrewCut/FX/NS_CrewCut_DamagedFence_2
dependencies {
 usedDataInterfaces [&NiagaraDataInterfaceArrayPosition]
 usedDynamicInputs [&RandomRangeFloat, /Niagara/DynamicInputs/Arrays/SelectPositionFromArray]
 usedModules [
  &SystemState
  &self:SystemSpawnScript.SystemScriptSource.SystemScriptGraph.NiagaraNodeAssignment_0.SetVariables_#1
  &self:SystemSpawnScript.SystemScriptSource.SystemScriptGraph.NiagaraNodeAssignment_1.SetVariables_#2
 ]
}
emitters {Airglow {rendererClasses [], renderers []}, Flare {rendererClasses [], renderers []}, Sparks {rendererClasses [], renderers []}}
summary {
 systemName=NS_CrewCut_DamagedFence_2
 emitters [{emitterName=Sparks, rendererClasses []}, {emitterName=Airglow, rendererClasses []}, {emitterName=Flare, rendererClasses []}]
 userVariables []
}
